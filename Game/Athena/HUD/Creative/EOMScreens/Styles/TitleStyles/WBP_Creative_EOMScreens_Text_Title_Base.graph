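# Base widget blueprint for title text styles on Creative EOM screens, derived directly from UserWidget.
# It declares a Play EOM Title Animation custom event that takes an InText parameter of type Text.
# The event has no logic attached, and the widget has no variables or child widgets, so it serves only as a shared interface.
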WidgetBlueprint /Game/Athena/HUD/Creative/EOMScreens/Styles/TitleStyles/WBP_Creative_EOMScreens_Text_Title_Base
graphs [EventGraph {nodes [Play EOM Title Animation!() -> (OutputDelegate, then, InText:Text) {is=title K2Node_CustomEvent}]}]
variables []
widgets {info {inheritedWidgetCount=0, parentClass=&UserWidget, rootWidgetClass=null, widgetCount=0}, widgets []}
