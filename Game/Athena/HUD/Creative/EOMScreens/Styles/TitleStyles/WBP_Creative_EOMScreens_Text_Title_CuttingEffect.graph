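# Creative EOM screen title style that shows the title being sliced in two by a cutting line.
# Play EOM Title Animation sets three overlaid text copies, hides the full one, shows the top and bottom ones, and plays Anim_Idle as a reset.
# A timer, cleared on replay and destruct, then plays Anim_TextCutting.
# That animation drives font material shape vertices to split the text, sweeps a reddening line material and slides the top piece aside.
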WidgetBlueprint /Game/Athena/HUD/Creative/EOMScreens/Styles/TitleStyles/WBP_Creative_EOMScreens_Text_Title_CuttingEffect
animations [
 Anim_Idle {
  binding_count=3
  length_seconds=0.05
  object=&self:Anim_Idle
  sequence {bindings [
    Image_CuttingLine {
     id=#1
     possessed_class=Image
     tracks [
      {
       type=MovieSceneWidgetMaterialTrack
       sections [{
         type=MovieSceneParameterSection
         end_frame=1
         start_frame=0
         channels [
          Shape Glow Min {type=MovieSceneScriptingFloatChannel}
          Shape Glow Max {type=MovieSceneScriptingFloatChannel}
          Shape Progress {type=MovieSceneScriptingFloatChannel, const [0,1]}
          Shape Progress Rotation {type=MovieSceneScriptingFloatChannel, const [0.5,1]}
          Shape Offset Y {type=MovieSceneScriptingFloatChannel}
          Shape Smooth Progress Max {type=MovieSceneScriptingFloatChannel, const [0,1]}
          Shape Smooth Progress Min {type=MovieSceneScriptingFloatChannel, const [0,1]}
          Shape Scale Y {type=MovieSceneScriptingFloatChannel}
          Shape Rotation {type=MovieSceneScriptingFloatChannel}
          Shape Color 1R {type=MovieSceneScriptingFloatChannel, const [1,1]}
          Shape Color 1G {type=MovieSceneScriptingFloatChannel, const [1,1]}
          Shape Color 1B {type=MovieSceneScriptingFloatChannel, const [1,1]}
          Shape Color 1A {type=MovieSceneScriptingFloatChannel, const [1,1]}
         ]
        }]
      }
      {
       type=MovieSceneByteTrack
       display_name=Visibility
       sections [{type=MovieSceneByteSection, end_frame=1, start_frame=0, channels [None {type=MovieSceneScriptingByteChannel, const [1,1]}]}]
      }
     ]
    }
    Text_Bottom {
     id=#2
     possessed_class=CommonTextBlock
     tracks [{
       type=MovieSceneWidgetMaterialTrack
       display_name=Font.FontMaterial
       sections [{
         type=MovieSceneParameterSection
         end_frame=1
         start_frame=0
         channels [
          Shape Vertex 1 X {type=MovieSceneScriptingFloatChannel, const [1,1]}
          Shape Vertex 1 Y {type=MovieSceneScriptingFloatChannel, const [-0.19,1]}
         ]
        }]
      }]
    }
    Text_Top {
     id=#3
     possessed_class=CommonTextBlock
     tracks [
      {
       type=MovieScene2DTransformTrack
       sections [{
         type=MovieScene2DTransformSection
         end_frame=1
         start_frame=0
         channels [
          Translation.X {type=MovieSceneScriptingFloatChannel, const [0,1]}
          Translation.Y {type=MovieSceneScriptingFloatChannel, const [0,1]}
          Angle {type=MovieSceneScriptingFloatChannel, const [0,1]}
         ]
        }]
      }
      {
       type=MovieSceneWidgetMaterialTrack
       display_name=Font.FontMaterial
       sections [{
         type=MovieSceneParameterSection
         end_frame=1
         start_frame=0
         channels [
          Shape Vertex 3 X {type=MovieSceneScriptingFloatChannel, const [1,2]}
          Shape Vertex 3 Y {type=MovieSceneScriptingFloatChannel, const [-0.19,1]}
         ]
        }]
      }
     ]
    }
   ]}
 }
 Anim_TextCutting {
  binding_count=3
  length_seconds=1.25
  object=&self:Anim_TextCutting
  sequence {bindings [
    Image_CuttingLine {
     id=#1
     possessed_class=Image
     tracks [
      {
       type=MovieSceneWidgetMaterialTrack
       sections [{
         type=MovieSceneParameterSection
         end_frame=25
         start_frame=0
         channels [
          Shape Glow Min {type=MovieSceneScriptingFloatChannel, const [-0.04,1]}
          Shape Glow Max {type=MovieSceneScriptingFloatChannel, const [0.035,1]}
          Shape Progress {type=MovieSceneScriptingFloatChannel, keys [[0,0],[2,0.9616],[6,0.75]]}
          Shape Progress Rotation {type=MovieSceneScriptingFloatChannel, keys [[0,0.5],[2,0]]}
          Shape Offset Y {type=MovieSceneScriptingFloatChannel, const [0.14,1]}
          Shape Smooth Progress Max {type=MovieSceneScriptingFloatChannel, keys [[0,0],[1,0.29],[3,-0.22]]}
          Shape Smooth Progress Min {type=MovieSceneScriptingFloatChannel, keys [[0,0],[1,-0.0151],[2,-0.45],[3,-1.7984]]}
          Shape Scale Y {type=MovieSceneScriptingFloatChannel, const [0.24,1]}
          Shape Rotation {type=MovieSceneScriptingFloatChannel, const [0.065,1]}
          Shape Color 1R {type=MovieSceneScriptingFloatChannel, const [1,2]}
          Shape Color 1G {type=MovieSceneScriptingFloatChannel, keys [[0,1],[2,0]]}
          Shape Color 1B {type=MovieSceneScriptingFloatChannel, keys [[0,1],[2,0]]}
          Shape Color 1A {type=MovieSceneScriptingFloatChannel, const [1,2]}
         ]
        }]
      }
      {
       type=MovieSceneByteTrack
       display_name=Visibility
       sections [{type=MovieSceneByteSection, end_frame=7, start_frame=0, channels [None {type=MovieSceneScriptingByteChannel, keys [[0,3],[7,1]]}]}]
      }
     ]
    }
    Text_Bottom {
     id=#2
     possessed_class=CommonTextBlock
     tracks [{
       type=MovieSceneWidgetMaterialTrack
       display_name=Font.FontMaterial
       sections [{
         type=MovieSceneParameterSection
         end_frame=25
         start_frame=0
         channels [
          Shape Vertex 1 X {type=MovieSceneScriptingFloatChannel, keys [[0,1],[2,-1]]}
          Shape Vertex 1 Y {type=MovieSceneScriptingFloatChannel, keys [[0,-0.19],[1,0],[2,0.14]]}
         ]
        }]
      }]
    }
    Text_Top {
     id=#3
     possessed_class=CommonTextBlock
     tracks [
      {
       type=MovieScene2DTransformTrack
       sections [{
         type=MovieScene2DTransformSection
         end_frame=25
         start_frame=2
         channels [
          Translation.X {type=MovieSceneScriptingFloatChannel, keys [[2,0],[7,-6.4443],[21,-18.6]]}
          Translation.Y {type=MovieSceneScriptingFloatChannel, keys [[2,0],[9,-0.26],[11,0]]}
          Angle {type=MovieSceneScriptingFloatChannel, keys [[2,0],[8,1],[11,0.1]]}
         ]
        }]
      }
      {
       type=MovieSceneWidgetMaterialTrack
       display_name=Font.FontMaterial
       sections [{
         type=MovieSceneParameterSection
         end_frame=25
         start_frame=0
         channels [
          Shape Vertex 3 X {type=MovieSceneScriptingFloatChannel, thinned_from=3, keys [[0,1],[2,-1]]}
          Shape Vertex 3 Y {type=MovieSceneScriptingFloatChannel, keys [[0,-0.19],[1,0],[2,0.14]]}
         ]
        }]
      }
     ]
    }
   ]}
 }
]
graphs [
 EventGraph {nodes [
   n1 = Event Play Text Cutting Animation() {is=title K2Node_CustomEvent}
   > PlayAnimation(InAnimation=Get Anim_TextCutting() -> (Anim_TextCutting:Widget Animation Object Reference))
   > Clear and Invalidate Timer by Handle(Handle=Get AnimationDelay() -> (AnimationDelay:Timer Handle Structure))
   n2 = Event Play EOM Title Animation!() -> (OutputDelegate, then, InText:Text) {is=title K2Node_Event}
   > Parent: Play EOM Title Animation!(execute, InText:Text=@n2.InText) -> (then) {is=title K2Node_CallParentFunction}
   > SetTitleText!(execute, self, InText:Text=@n2.InText) -> (then)
   > StopAllAnimations(self:User Widget Object Reference)
   > Clear and Invalidate Timer by Handle(Handle=Get AnimationDelay())
   > UpdateTextVisibility!(
    execute
    self
    InFullTextVisibility:ESlateVisibility Enum=Collapsed
    InTopTextVisibility:ESlateVisibility Enum=HitTestInvisible
    InBottomTextVisibility:ESlateVisibility Enum=HitTestInvisible
   ) -> (then)
   > n3 = "Play Animation with Finished event (legacy)"!(
    execute:Exec
    Widget:User Widget Object Reference=Self-Reference()
    InAnimation:Widget Animation Object Reference=Get Anim_Idle() -> (Anim_Idle:Widget Animation Object Reference)
    StartAtTime:float=0
    NumLoopsToPlay:int=1
    PlayMode:EUMGSequencePlayMode Enum
    PlaybackSpeed:float=1
   ) -> (then:Exec, Result:UMGSequence Player Object Reference, Finished:Exec) {is=title K2Node_PlayAnimation}
   n4 = Set Timer by Event(execute=@n3.Finished, Delegate=@n1, Time=1.37, bLooping=false)
   > Set AnimationDelay(AnimationDelay:Timer Handle Structure=@n4) -> (Output_Get:Timer Handle Structure)
   Event Destruct()
   > Clear and Invalidate Timer by Handle(Handle=Get AnimationDelay())
   Event Construct()
   > UpdateTextVisibility(InFullTextVisibility=HitTestInvisible, InTopTextVisibility=Collapsed, InBottomTextVisibility=Collapsed)
  ]}
 Sequencer Events {links [], nodes []}
 SetTitleText {nodes [
   n1 = SetTitleText!() -> (then, InText:Text) {is=title K2Node_FunctionEntry}
   > "SetText (Text)"(
    self=[
     Get Text_Full() -> (Text_Full:Common Text Object Reference)
     Get Text_Top() -> (Text_Top:Common Text Object Reference)
     Get Text_Bottom() -> (Text_Bottom:Common Text Object Reference)
    ]
    InText=@n1
   )
  ]}
 UpdateTextVisibility {nodes [
   UpdateTextVisibility!() -> (
    then
    InFullTextVisibility:ESlateVisibility Enum=Visible
    InTopTextVisibility:ESlateVisibility Enum=Visible
    InBottomTextVisibility:ESlateVisibility Enum=Visible
   ) {is=title K2Node_FunctionEntry}
   > SetVisibility(self=Get Text_Full(), InVisibility=Get InFullTextVisibility() -> (InFullTextVisibility:ESlateVisibility Enum=Visible))
   > SetVisibility(self=Get Text_Top(), InVisibility=Get InTopTextVisibility() -> (InTopTextVisibility:ESlateVisibility Enum=Visible))
   > SetVisibility(self=Get Text_Bottom(), InVisibility=Get InBottomTextVisibility() -> (InBottomTextVisibility:ESlateVisibility Enum=Visible))
  ]}
]
variables [AnimationDelay {type=TimerHandle}]
widgets {
 info {
  inheritedWidgetCount=0
  parentClass=&/WBP_Creative_EOMScreens_Text_Title_Base.WBP_Creative_EOMScreens_Text_Title_Base_C
  rootWidgetClass=&Overlay
  widgetCount=5
 }
 widgets [
  Root {widgetClassPath=&Overlay, widget=Root}
  Text_Top {
   widgetClassPath=&CommonTextBlock
   bIsVariable=true
   parent=Root
   widget=Text_Top
   properties {
    justification=Center
    text="[PH] TITLE CUT"
    textTransformPolicy=ToUpper
    visibility=Collapsed
    font {
     bForceMonospaced=false
     bMaterialIsStencil=false
     fontMaterial=/Game/Athena/HUD/Creative/EOMScreens/MaterialInstances/MI_UI_Creative_EOMScreens_Title_TextCutting_Top
     fontObject=&Burbank
     letterSpacing=0
     monospacedWidth=1
     size=72
     skewAmount=0.2
     typefaceFontName=Burbank-Black
     outlineSettings {
      bApplyOutlineToDropShadows=false
      bMiteredCorners=false
      bSeparateFillAlpha=false
      outlineMaterial=null
      outlineSize=0
      outlineColor=<rgba 0 0 0 1>
     }
    }
    shadowOffset=<0 0>
   }
   slot {class=OverlaySlot, properties {horizontalAlignment=HAlign_Fill, verticalAlignment=VAlign_Center}}
  }
  Text_Bottom {
   widgetClassPath=&CommonTextBlock
   bIsVariable=true
   parent=Root
   widget=Text_Bottom
   properties {
    justification=Center
    text="[PH] TITLE CUT"
    textTransformPolicy=ToUpper
    visibility=Collapsed
    font {
     bForceMonospaced=false
     bMaterialIsStencil=false
     fontMaterial=/Game/Athena/HUD/Creative/EOMScreens/MaterialInstances/MI_UI_Creative_EOMScreens_Title_TextCutting_Bottom
     fontObject=&Burbank
     letterSpacing=0
     monospacedWidth=1
     size=72
     skewAmount=0.2
     typefaceFontName=Burbank-Black
     outlineSettings {
      bApplyOutlineToDropShadows=false
      bMiteredCorners=false
      bSeparateFillAlpha=false
      outlineMaterial=null
      outlineSize=0
      outlineColor=<rgba 0 0 0 1>
     }
    }
    shadowOffset=<0 0>
   }
   slot {class=OverlaySlot, properties {horizontalAlignment=HAlign_Fill, verticalAlignment=VAlign_Center}}
  }
  Text_Full {
   widgetClassPath=&CommonTextBlock
   bIsVariable=true
   parent=Root
   widget=Text_Full
   properties {
    justification=Center
    text="[PH] TITLE CUT"
    textTransformPolicy=ToUpper
    font {
     bForceMonospaced=false
     bMaterialIsStencil=false
     fontMaterial=null
     fontObject=&Burbank
     letterSpacing=0
     monospacedWidth=1
     size=72
     skewAmount=0.2
     typefaceFontName=Burbank-Black
     outlineSettings {
      bApplyOutlineToDropShadows=false
      bMiteredCorners=false
      bSeparateFillAlpha=false
      outlineMaterial=null
      outlineSize=0
      outlineColor=<rgba 0 0 0 1>
     }
    }
    shadowOffset=<0 0>
   }
   slot {class=OverlaySlot, properties {horizontalAlignment=HAlign_Fill, verticalAlignment=VAlign_Center}}
  }
  Image_CuttingLine {
   widgetClassPath=&Image
   bIsVariable=true
   parent=Root
   widget=Image_CuttingLine
   properties {
    visibility=Collapsed
    brush {
     bIsDynamicallyLoaded=false
     drawAs=Image
     imageType=NoImage
     mirroring=NoMirror
     resourceName=None
     resourceObject=/Game/Athena/HUD/Creative/EOMScreens/MaterialInstances/MI_UI_Creative_EOMScreens_Title_TextCutting_Line
     tiling=NoTile
     imageSize=<32 32>
     margin=<ltrb 0 0 0 0>
     outlineSettings {
      bUseBrushTransparency=false
      roundingType=HalfHeightRadius
      width=0
      color {colorUseRule=UseColor_Specified, specifiedColor=<rgba 0 0 0 0>}
      cornerRadii=<0 0 0 0>
     }
     tintColor {colorUseRule=UseColor_Specified, specifiedColor=<rgba 1 1 1 1>}
     uVRegion {bIsValid=false, max=<0 0>, min=<0 0>}
    }
    renderTransform {angle=0, scale=<3 3>, shear=<0 0>, translation=<0 0>}
   }
   slot {class=OverlaySlot, properties {horizontalAlignment=HAlign_Fill, verticalAlignment=VAlign_Fill}}
  }
 ]
}
