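# Simple title style widget for Creative EOM screens, derived from the shared EOM title base widget.
# It consists of a single centered, uppercase, skewed Burbank Black text block.
# The Play EOM Title Animation event calls the parent version, then sets the text through SetTitleText, with no animation of its own.
# A Reset Text Display event exists but has no logic attached.
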WidgetBlueprint /Game/Athena/HUD/Creative/EOMScreens/Styles/TitleStyles/WBP_Creative_EOMScreens_Text_Title_Simple
graphs [
 EventGraph {nodes [
   n1 = Event Play EOM Title Animation!() -> (OutputDelegate, then, InText:Text) {is=title K2Node_Event}
   > Parent: Play EOM Title Animation!(execute, InText:Text=@n1.InText) -> (then) {is=title K2Node_CallParentFunction}
   > SetTitleText!(execute, self, InText:Text=@n1.InText) -> (then)
   Reset Text Display() {is=title K2Node_CustomEvent}
  ]}
 SetTitleText {nodes [
   n1 = SetTitleText!() -> (then, InText:Text) {is=title K2Node_FunctionEntry}
   > "SetText (Text)"(self=Get Text_Title() -> (Text_Title:Common Text Object Reference), InText=@n1)
  ]}
]
variables []
widgets {
 info {
  inheritedWidgetCount=0
  parentClass=&/WBP_Creative_EOMScreens_Text_Title_Base.WBP_Creative_EOMScreens_Text_Title_Base_C
  rootWidgetClass=&CommonTextBlock
  widgetCount=1
 }
 widgets [Text_Title {
   widgetClassPath=&CommonTextBlock
   bIsVariable=true
   widget=Text_Title
   properties {
    justification=Center
    text="[PH] TITLE"
    textTransformPolicy=ToUpper
    font {
     bForceMonospaced=false
     bMaterialIsStencil=false
     fontMaterial=null
     fontObject=&Burbank
     letterSpacing=0
     monospacedWidth=1
     size=72
     skewAmount=0.2
     typefaceFontName=Burbank-Black
     outlineSettings {
      bApplyOutlineToDropShadows=false
      bMiteredCorners=false
      bSeparateFillAlpha=false
      outlineMaterial=null
      outlineSize=0
      outlineColor=<rgba 0 0 0 1>
     }
    }
    shadowOffset=<0 0>
   }
  }]
}
